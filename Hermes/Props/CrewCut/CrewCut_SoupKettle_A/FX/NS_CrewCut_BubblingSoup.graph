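# NS_CrewCut_BubblingSoup is a Niagara system in the CrewCut soup kettle prop's FX folder, named for a bubbling soup effect.
# It holds two emitters, Empty and Empty001, and neither lists any renderer.
# SystemState is the only module listed, with no data interfaces, dynamic inputs or user variables.
# As recorded here it draws nothing, so it appears to be an unfinished placeholder.
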NiagaraSystem /Hermes/Props/CrewCut/CrewCut_SoupKettle_A/FX/NS_CrewCut_BubblingSoup
dependencies {usedDataInterfaces [], usedDynamicInputs [], usedModules [&SystemState]}
emitters {Empty {rendererClasses [], renderers []}, Empty001 {rendererClasses [], renderers []}}
summary {
 systemName=NS_CrewCut_BubblingSoup
 emitters [{emitterName=Empty, rendererClasses []}, {emitterName=Empty001, rendererClasses []}]
 userVariables []
}
